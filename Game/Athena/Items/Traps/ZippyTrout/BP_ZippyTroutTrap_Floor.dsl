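; ZippyTrout floor trap Blueprint, handling the trap's trigger setup, audio, lightning effects and team-based material in Athena.
; Placement adds the trigger, sound and dust; a replicated finished-building flag starts idle particles.
; The ActivateTrap gameplay cue drives a timestamped material jiggle, plays lightning, briefly pauses idle particles and fades loop audio until reload.
; Destruction plays explosion effects; team init gives same-team local players a friendly material.
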(event EventOnPlaced
  ;[1 1 1] Initial Setup
  (Development|PrintString (Utilities|String|ToString(Boolean) (Utilities|Enum|Equal(Enum) (Fortnite|GetSubGame) "Athena")) true true "(R=1.000000,G=0.010197,B=0.030093,A=1.000000)")
  (Utilities|FlowControl|DoOnce false
    (:Completed
      (Utilities|FlowControl|Sequence
        (:then_0
          (Variables|Default|SetMuzzleSockets (Transformation|GetAllSocketNames (Variables|Debug|GetStaticMeshComponent))))
        (:then_1
          (Trap|AddTriggerComponent (Variables|Trap_Floor_Spikes|GetTrigger))
          (if (not (Variables|Default|GetbPlayFewerSounds))
            (Audio|PlaySoundatLocation (Variables|Audio|GetPlace_Trap_Sound) (Transformation|GetActorLocation))
            (Fortnite|ShowSoundIndicatorLocalforTeam :Location (Transformation|GetActorLocation) :MaxAudibleDistance (Variables|/Script/Engine.SoundBase|GetMaxDistance (Variables|Audio|GetPlace_Trap_Sound)) :IndicatorType "Gunshot" :Team (Team|GetActorTeam self) :Affiliations (Utilities|Array|MakeArray "Hostile" "Neutral") :Tint "(R=0.000000,G=0.000000,B=0.000000,A=0.000000)" :OverrideIcon "/Game/UI/Foundation/Textures/Icons/Navigation/t_icon_nav_mobile_audioviz_building_32.t_icon_nav_mobile_audioviz_building_32" :AttenuationMaxRangeMultiplier 1.0 :AttenuationIntensityMultiplier 1.0)
            (Utilities|TrapPlacementDustFX :Attach%20to%20Component (Variables|Debug|GetStaticMeshComponent) :Socket "PlacedFX" :Rotation "0, 0, 0" :Burst%20Rate 15 :Size%20%28Volume%29 "200.000000,200.000000,15.000000" :Velocity%20Scale "1,1,1" :Size%20Scale%20%28Particles%29 1)))))))

(event Custom|Jiggle
  (params (Vector :type Vector))
  ;[1 1 1] Initial Setup
  (Rendering|Material|SetVectorParameterValueonMaterials (Variables|Debug|GetStaticMeshComponent) "AmountVector" Vector))

(event EventOnFinishedBuilding
  ;[1 1 1] Finished building
  (Variables|Default|SetbIsFinishedBuilding true))

(event EventOnWorldReady
  ;[1 1 1] Finished building
  (Networking|CanExecuteCosmeticEvents
    (:True
      (if (not (Variables|Default|GetbPlayFewerSounds))
        (bind _returnvalue (Audio|SpawnSoundAttached (Variables|Audio|GetTrap_Active_Sound) (Variables|Debug|GetStaticMeshComponent)))
        (Variables|Default|SetActive%20Loop%20Audio _returnvalue)
        (Fortnite|ShowSoundIndicatorLocalforTeam :Location (Transformation|GetActorLocation) :MaxAudibleDistance (Variables|/Script/Engine.SoundBase|GetMaxDistance (Variables|Audio|GetPlace_Trap_Sound)) :IndicatorType "Gunshot" :Team (Team|GetActorTeam self) :Affiliations (Utilities|Array|MakeArray "Hostile" "Neutral") :Tint "(R=0.000000,G=0.000000,B=0.000000,A=0.000000)" :OverrideIcon "/Game/UI/Foundation/Textures/Icons/Navigation/t_icon_nav_mobile_audioviz_building_32.t_icon_nav_mobile_audioviz_building_32" :AttenuationMaxRangeMultiplier 1.0 :AttenuationIntensityMultiplier 1.0)))))

(event GameplayCue|GameplayCue.Abilities.Activation.Traps.ActivateTrap
  (params (EventType :type EGameplayCueEvent%20Enum :default "OnActive") (Parameters :type Gameplay%20Cue%20Parameters%20Structure%20%28by%20ref%29))
  (bind _returnvalue (Utilities|Time|GetGameTimeinSeconds))
  (bind _staticmeshcomponent (Variables|Debug|GetStaticMeshComponent))
  ;[1 1 1] On Fire
  (switch Utilities|FlowControl|Switch|SwitchonEGameplayCueEvent EventType
    (:Executed
      (Audio|PlaySoundatLocation (Variables|Audio|GetTrap_Fire_Sound) (Transformation|GetActorLocation))
      (Rendering|Material|SetScalarParameterValueonMaterials _staticmeshcomponent "TimestampA" _returnvalue)
      (Rendering|Material|SetScalarParameterValueonMaterials _staticmeshcomponent "TimestampB" (+ _returnvalue (Variables|Default|GetTimespan%20of%20Jiggle)))
      (CallFunction|MuzzleFire "None")
      (CallFunction|Jiggle (Variables|Default|GetVector%20of%20Jiggle))
      (bind _active_loop_audio (Variables|Default|GetActive%20Loop%20Audio)
        (:then
          (Audio|Components|Audio|FadeOut _active_loop_audio 0.1 0.0 "Linear" :!class /Script/Engine.AudioComponent))
        (:else)))))

(event Custom|MuzzleFire
  (params (Socket%20Name :type Name))
  (bind _p_zippytrout_idle (Variables|Default|GetP_ZippyTrout_Idle))
  ;[1 1 1] On Fire
  (Networking|CanExecuteCosmeticEvents
    (:True
      (Components|Activation|Activate (Variables|Default|GetP_ZippyTrout_Lightning) true)
      (Components|Activation|Deactivate _p_zippytrout_idle)
      (Utilities|FlowControl|RetriggerableDelay 2.0)
      (Components|Activation|Activate _p_zippytrout_idle true))))

(event EventOnReloadEnd
  ;[1 1 1] Reload End
  (bind _returnvalue (Audio|SpawnSoundAttached (Variables|Audio|GetTrap_Active_Sound) (Variables|Debug|GetStaticMeshComponent)))
  (Variables|Default|SetActive%20Loop%20Audio _returnvalue))

(event EventDestroyed
  ;[1 1 1] Durability Change
  (bind _active_loop_audio (Variables|Default|GetActive%20Loop%20Audio)
    (:then
      (Audio|Components|Audio|FadeOut _active_loop_audio 0.1 0.0 "Linear" :!class /Script/Engine.AudioComponent)
      (Audio|PlaySoundatLocation (Variables|Audio|GetTrap_Explode_Sound) (Transformation|GetActorLocation))
      (Effects|SpawnEmitteratLocation (Variables|Default|GetDestructionFX) (Transformation|GetWorldLocation (Variables|Transformation|GetRootComponent))))
    (:else)))

(event EventOnInitTeam
  ;[1 1 1] Detect if local player is on the same team as this trap
  (CallFunction|ChangetoFriendlyMaterial))

(fn ChangeToFriendlyMaterial
  (bind _returnvalue (Fortnite|GetLocalFortPlayerControllers true))
  (if (and (Utilities|Enum|Equal(Enum) (Fortnite|GetSubGame) "Athena") (> (Utilities|Array|Length _returnvalue) 0))
    (if (Math|Byte|Equal(Byte) (Variables|Team|GetTeamIndex) (Team|GetActorTeam (Utilities|Array|Get(acopy) _returnvalue)))
      (Rendering|Material|SetMaterial (Variables|Debug|GetStaticMeshComponent) 0 (Variables|Default|GetFriendlyTrapMaterial)))))

(fn ChangeToNormalMaterial
  (Rendering|Material|SetMaterial (Variables|Debug|GetStaticMeshComponent) 0 (Variables|Default|GetNormalTrapMaterial)))

(fn OnRep_bIsFinishedBuilding
  (Components|Activation|Activate (Variables|Default|GetP_ZippyTrout_Idle) true))

(fn ConstructionScript)
