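; Thrown trap grenade projectile (ZippyTrout, with Zapper trap throw sound) that places a trap where it lands.
; On hit, the server stores the hit result, updates the deco helper and stops and hides the projectile at valid spots.
; The hit result's OnRep lets non-authority clients attempt placement through the deco tool, then explode and hide.
; Invalid hits record throttled bounces, with a one-time delayed retry.
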(event Custom|HandleBounce
  (params (Hit%20Location :type Vector))
  (if (and (> (Math|Vector|VectorLengthSquared (- Hit%20Location (Variables|Default|GetLastBounceLocation))) (Math|Float|Square (Variables|Default|GetMinBounceDistance))) (> (- (Utilities|Time|GetTimeSeconds) (Variables|Default|GetLastBounceTime)) (Variables|Default|GetMinTimeBetweenBounces)))
    (Variables|Default|SetLastBounceLocation Hit%20Location)
    (Variables|Default|SetLastBounceTime (Utilities|Time|GetTimeSeconds))))

(event Collision|EventHit
  (params (MyComp :type Primitive%20Component%20Object%20Reference) (Other :type Actor%20Object%20Reference) (OtherComp :type Primitive%20Component%20Object%20Reference) (bSelfMoved :type Boolean) (HitLocation :type Vector :default "0, 0, 0") (HitNormal :type Vector :default "0, 0, 0") (NormalImpulse :type Vector :default "0, 0, 0") (Hit :type Hit%20Result%20Structure%20%28by%20ref%29))
  (bind _returnvalue (Actor|GetOwner))
  (bind (_asbuilding_prop _bsuccess) (Utilities|Casting|CastToBuildingProp (Variables|Default|GetHitActor)))
  (bind (_asathena_fill_floor _bsuccess_2) (Utilities|Casting|CastToAthenaFillFloor (Variables|Default|GetHitActor)))
  (bind (_asfort_athena_supply_drop _bsuccess_3) (Utilities|Casting|CastToFortAthenaSupplyDrop (Variables|Default|GetHitActor)))
  (bind (_asfort_projectile_base _bsuccess_4) (Utilities|Casting|CastToFortProjectileBase (Variables|Default|GetHitActor)))
  (bind (_asathena_supply_drop_balloon _bsuccess_5) (Utilities|Casting|CastToAthenaSupplyDropBalloon (Variables|Default|GetHitActor)))
  (bind (_asfort_pawn _bsuccess_6) (Utilities|Casting|CastToFortPawn (Variables|Default|GetHitActor)))
  (|CallParentFunction MyComp Other OtherComp bSelfMoved HitLocation HitNormal NormalImpulse Hit)
  (Variables|Default|SetHitActor Other)
  (Utilities|IsValid _returnvalue
    (:Is%20Valid
      (bind _output_get (Variables|Default|SetGrenadeHitResult Hit))
      (if (Networking|HasAuthority _returnvalue)
        ;[1 1 1] Check hit on server side, if it is a valid build location stop simulation and wait for client to catch up.
        (CallFunction|SetHitandTick _output_get)
        (if (not (Math|Boolean|ORBoolean (and (not _bsuccess_2) (Math|Boolean|ORBoolean _bsuccess _bsuccess_6 _bsuccess_3 _bsuccess_5 _bsuccess_4)) (Class|FortProjectileDecoHelper|GetInvalidThrowTrapPosition (Variables|FortTrapGrenadeProjectile|GetProjectileDecoHelper)) (not (Class|FortDecoHelper|IsinValidPlacementSpot (Variables|FortTrapGrenadeProjectile|GetProjectileDecoHelper) false))))
          (Game|Components|ProjectileMovement|StopSimulating (Variables|Movement|GetProjectileMovementComponent) _output_get)
          (Rendering|SetActorHiddenInGame true))))))

(event Custom|GrenadeHitClient
  (bind _grenadehitresult (Variables|Default|GetGrenadeHitResult))
  (bind (_bblockinghit _binitialoverlap _time _distance _location _impactpoint _normal _impactnormal _physmat _hitactor) (Collision|BreakHitResult _grenadehitresult))
  (bind _returnvalue (Actor|GetOwner))
  (bind (_asbuilding_prop _bsuccess) (Utilities|Casting|CastToBuildingProp _hitactor))
  (bind (_asathena_fill_floor _bsuccess_2) (Utilities|Casting|CastToAthenaFillFloor _hitactor))
  (bind (_asfort_athena_supply_drop _bsuccess_3) (Utilities|Casting|CastToFortAthenaSupplyDrop _hitactor))
  (bind (_asfort_projectile_base _bsuccess_4) (Utilities|Casting|CastToFortProjectileBase _hitactor))
  (bind (_asathena_supply_drop_balloon _bsuccess_5) (Utilities|Casting|CastToAthenaSupplyDropBalloon _hitactor))
  (bind (_asfort_pawn _bsuccess_6) (Utilities|Casting|CastToFortPawn _hitactor))
  (Utilities|IsValid _returnvalue
    (:Is%20Valid
      (if (Math|Boolean|NORBoolean (Variables|Default|GetTrapPlaced) (Networking|HasAuthority _returnvalue))
        (Utilities|IsValid (Variables|FortTrapGrenadeProjectile|GetProjectileDecoHelper)
          (:Is%20Valid
            (Utilities|IsValid (Variables|FortTrapGrenadeProjectile|GetProjectileDecoTool)
              (:Is%20Valid
                (CallFunction|SetHitandTick _grenadehitresult)
                (if (or (and (not _bsuccess_2) (Math|Boolean|ORBoolean _bsuccess _bsuccess_6 _bsuccess_3 _bsuccess_5 _bsuccess_4)) (Class|FortProjectileDecoHelper|GetInvalidThrowTrapPosition (Variables|FortTrapGrenadeProjectile|GetProjectileDecoHelper)))
                  (CallFunction|HandleBounce _location)
                  (Utilities|FlowControl|DoOnce false
                    (:Completed
                      (Utilities|FlowControl|Delay 5.0)
                      (CallFunction|SetHitandTick _grenadehitresult))))
                (bind _returnvalue_2 (Class|FortDecoTool|TrytoPlace (Variables|FortTrapGrenadeProjectile|GetProjectileDecoTool)))
                (if (not _returnvalue_2)
                  (CallFunction|HandleBounce _location)
                  (Utilities|FlowControl|DoOnce false
                    (:Completed
                      (Utilities|FlowControl|Delay 5.0)
                      (CallFunction|SetHitandTick _grenadehitresult)
                      (Class|FortDecoTool|TrytoPlace (Variables|FortTrapGrenadeProjectile|GetProjectileDecoTool)))))
                (Utilities|FlowControl|DoOnce false
                  (:Completed
                    (Variables|Default|SetTrapPlaced true)
                    (Class|FortDecoTool|OnUnEquip (Variables|FortTrapGrenadeProjectile|GetProjectileDecoTool))
                    (Delivery|DoExplosion)
                    (Game|Components|ProjectileMovement|StopSimulating (Variables|Movement|GetProjectileMovementComponent) _grenadehitresult)
                    (Rendering|SetActorHiddenInGame true)))))))))))

(event EventEndPlay
  (params (EndPlayReason :type EEndPlayReason%20Enum :default "Destroyed"))
  (Class|FortDecoTool|OnUnEquip (Variables|FortTrapGrenadeProjectile|GetProjectileDecoTool))
  (Rendering|SetActorHiddenInGame true))

(event EventBeginPlay
  ;[1 1 1] Play Throw Sound
  (Audio|SpawnSoundAttached "/Game/Athena/Sounds/Items/Zapper_Trap_V2/Zapper_Trap_V2_Throw_Cue.Zapper_Trap_V2_Throw_Cue" (Variables|Collision|GetCapsuleComponent)))

(fn OnRep_GrenadeHitResult
  (CallFunction|GrenadeHitClient))

(fn SetHitAndTick
  (params (Hit :type Hit%20Result%20Structure))
  (bind _projectiledecohelper (Variables|FortTrapGrenadeProjectile|GetProjectileDecoHelper))
  (Class|FortProjectileDecoHelper|SetGrenadeTrajectoryHitResult _projectiledecohelper Hit)
  (Class|FortProjectileDecoHelper|Tick _projectiledecohelper (Utilities|Time|GetWorldDeltaSeconds)) ; Tick through helper one more time to make sure values are updated to new hit location
)

(fn ConstructionScript)
