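# Translucent unlit mesh-particle material for flickering orange lightning, named for the Boardwalk pickaxe idle effect.
# World position offset combines panning normal-map noise, shaped by a sine along V, with SplineThicken width from a parameter or particle life.
# Emissive blends two warm tones through a U-centered mask, tinted by particle color.
# Opacity pulses with a time sine, optionally masked around both V ends, while dynamic parameters tune noise and mask radius.
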Material /Game/Athena/Items/Traps/ZippyTrout/FX/M_Boardwalk_Pickaxe_Ilde_lightning
graph {
 properties {
  bUsedWithMeshParticles=true
  bUsedWithNiagaraMeshParticles=true
  blendMode=BLEND_Translucent
  pixelDepthOffsetMode=PDOM_Legacy
  shadingModel=MSM_Unlit
  twoSided=true
  usedShadingModels=Unlit
 }
 nodes [
  n1 = TextureCoordinate()
  n2 = Sine(None=Divide(A=ComponentMask(None=@n1) {g=true}))
  n3 = DynamicParameter() {paramNames ["WithRandom(0.8-1)", NoiseSpeedMult, "Lightning_complexity(0-1)", Param4]}
  n4 = Multiply(A=@n1, B=AppendVector(A=Constant() {r=0.02}, B=@n3."Lightning_complexity(0-1)"))
  n5 = Multiply(A=Constant2Vector() {g=0.02, r=0.7}, B=@n3.NoiseSpeedMult)
  n6 = ComponentMask(None=TextureCoordinate()) {r=true}
  n7 = Multiply(
   A=TextureSample!(UVs=Panner(Coordinate=@n4, Speed=@n5) {speedX=1, speedY=0.2}, Tex, Level, Apply View MipBias) {mipValueMode=TMVM_MipLevel, samplerType=SAMPLERTYPE_Normal, texture=/Game/Global/Textures/Default/Noise/T_Normal_Noise01}.RGB
   B=Multiply(A=LinearInterpolate(A=@n3."WithRandom(0.8-1)", Alpha=@n6) {constA=0.6}) {constB=3}
  )
  n8 = Multiply(A=ScalarParameter() {defaultValue=2, parameterName=Size}, B=ParticleRelativeTime())
  n9 = ScalarParameter() {defaultValue=0.4, parameterName=Param_3}
  n10 = TextureCoordinate()
  n11 = Sine(None=Multiply(A=ComponentMask(None=@n10) {g=true}) {constB=0.5})
  Add(
   A=Multiply(A=Multiply(A=@n2, B=@n7), B=ScalarParameter() {defaultValue=2, parameterName=WPO_intensity})
   B=MaterialFunctionCall(WidthBase=StaticSwitchParameter(True=@n8, False=@n9) {parameterName=Use particle life for size}, WidthTip=Constant(), UVs for Thickness=@n11) {materialFunction=&SplineThicken}.WorldPositionOffset
  )
  n12 = Constant3Vector() {constant=<rgba 1 0.3419 0.1 0>}
  n13 = Constant3Vector() {constant=<rgba 1 0.7637 0.4799 0>}
  n14 = ComponentMask(None=TextureCoordinate()) {r=true}
  n15 = Multiply(A=OneMinus(None=@n14), B=@n14)
  n16 = ScalarParameter() {defaultValue=3, group=Color, parameterName=Mask Exponent}
  n17 = Clamp(None=Multiply(A=Power(Base=@n15, Exp=@n16) {constExponent=3}, B=ScalarParameter() {defaultValue=50, group=Color, parameterName=Mask Scale}) {constB=50})
  MaterialFunctionCall(Emissive=Multiply(A=ParticleColor().RGB, B=LinearInterpolate(A=@n12, B=@n13, Alpha=@n17))) {materialFunction=&MF_FX_EmissiveAdjustments}
  n18 = Multiply(A=Time(), B=ScalarParameter() {defaultValue=7.5, parameterName=Flash_speed}) {constB=7.5}
  n19 = Divide(A=Add(A=Sine(None=@n18)))
  n20 = LinearInterpolate(Alpha=@n19) {constA=0.85}
  n21 = ComponentMask(None=TextureCoordinate()) {g=true}
  n22 = Constant()
  n23 = DynamicParameter() {paramNames ["WithRandom(0.8-1)", NoiseSpeedMult, "Lightning_complexity(0-1)", Param4]}
  n24 = ComponentMask(None=TextureCoordinate()) {g=true}
  n25 = Constant() {r=1}
  n26 = Clamp(None=Add(
    A=SphereMask(A=@n21, B=@n22, Radius=@n23.Param4) {attenuationRadius=0.5, hardnessPercent=0}
    B=SphereMask(A=@n24, B=@n25, Radius=@n23.Param4) {attenuationRadius=0.5, hardnessPercent=0}
   ))
  Multiply(A=StaticSwitchParameter(True=@n19, False=Multiply(A=@n20, B=@n26)) {defaultValue=true, parameterName=Mask}, B=ParticleColor().A)
  Multiply(A=@n11, B=ScalarParameter() {defaultValue=0.3})
  Power(Base=@n26)
 ]
}
